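# Translucent unlit Niagara sprite and mesh material for spline effects on the PaddleHarp Gap backpack.
# A channel-selected texture, with per-particle tiling and offset plus rotation, multiplies particle color for emissive and particle alpha for depth-faded opacity.
# World position offset sums two parameter-blended effects: SplineThicken scaled by per-particle weight, and a local-Z wobble from a mask texture and clamped time sine with per-particle random frequency.
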Material /BRCosmetics/Accessories/FORT_Backpacks/Backpack_F_MED_PaddleHarp_Gap/FX/M_Backpack_PaddleHarp_Gap_Spline
graph {
 properties {
  bAutomaticallySetUsageInEditor=false
  bUsedWithNiagaraMeshParticles=true
  bUsedWithNiagaraSprites=true
  blendMode=BLEND_Translucent
  shadingModel=MSM_Unlit
  usedShadingModels=Unlit
 }
 nodes [
  n1 = Multiply(
   A=ScalarParameter() {defaultValue=1, parameterName=Spline_ThickMul}
   B=DynamicParameter() {parameterIndex=1, paramNames [Spline_Weight, Param2, Param3, Param4]}.Spline_Weight
  )
  n2 = MaterialFunctionCall(
   WidthBase=ScalarParameter() {defaultValue=3, parameterName=Spline_Base}
   WidthTip=ScalarParameter() {defaultValue=3, parameterName=Spline_Tip}
   Expand U or V UV Channel=StaticBool() {value=true}
  ) {materialFunction=&SplineThicken}
  n3 = Panner(Coordinate=TextureCoordinate())
  n4 = Sine(None=Multiply(A=Time(), B=DynamicParameter() {paramNames [SubUV_TilingU, SubUV_UVAdd, WPO_Random, Param4]}.WPO_Random))
  n5 = Multiply(
   A=Transform(None=Constant3Vector() {constant=<rgba 0 0 1 0>}) {transformSourceType=TRANSFORMSOURCE_Local}
   B=Multiply(
    A=TextureSample(UVs=@n3) {texture=/BRCosmetics/Effects/Fort_Effects/Emotes/MarchTreatCup/T_Emote_MarchTreatCup_Mask}.RGB
    B=Clamp(None=@n4) {maxDefault=1.5}
   )
  )
  n6 = ScalarParameter() {defaultValue=1, parameterName=WPO_Intensity}
  Add(
   A=LinearInterpolate(A=Constant(), B=Multiply(A=@n1, B=@n2.WorldPositionOffset), Alpha=ScalarParameter() {parameterName=Use_SplineThicken})
   B=LinearInterpolate(B=Multiply(A=@n5, B=@n6), Alpha=ScalarParameter() {parameterName=Use_WPO})
  )
  n7 = ParticleColor()
  n8 = DynamicParameter() {paramNames [SubUV_TilingU, SubUV_UVAdd, WPO_Random, Param4]}
  n9 = Multiply(A=TextureCoordinate(), B=AppendVector(A=@n8.SubUV_TilingU, B=ScalarParameter() {defaultValue=1, parameterName=SubUV_TilingV}))
  n10 = AppendVector(A=@n8.SubUV_UVAdd, B=Constant())
  n11 = TextureSampleParameter2D(UVs=MaterialFunctionCall(UVs=Add(A=@n9, B=@n10), "Rotation Angle (0-1)"=ScalarParameter() {parameterName=TextureAngle}) {materialFunction=&CustomRotator}.Rotated Values) {parameterName=SubUV_Tex, texture=&/T_Backpack_PaddleHarp_Gap_Texture}
  n12 = Multiply(A=ChannelMaskParameter(None=@n11.RGBA) {parameterName=RGBA_Change}.R)
  DepthFade(Opacity=Multiply(A=@n7.A, B=@n12), FadeDistance=ScalarParameter() {defaultValue=2, parameterName=DepthFade})
  MaterialFunctionCall(Emissive=Multiply(A=@n7.RGB, B=@n12)) {materialFunction=&MF_FX_EmissiveAdjustments}
 ]
}
