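# Ice material for the PaddleHarp Gap back bling, adding misty, cracked-ice glow to the character parent material, with an AsPickaxe switch for pickaxe variants.
# MF_Ice_TechArt output gets emissive and base color overrides, selected through Tech Art mask channels.
# Bump-offset mist and crack-normal layers give parallax depth, while a custom directional-light node supports fake specular and subsurface terms.
# Frontend-loop audio amplitude pulses tinted emissive when AsPickaxe is off.
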Material /BRCosmetics/Accessories/FORT_Backpacks/Backpack_F_MED_PaddleHarp_Gap/Materials/M_PaddleHarp_Gap_Ice
graph {
 properties {
  bUseMaterialAttributes=true
  bUsedWithMeshDeformer=true
  bUsedWithNiagaraMeshParticles=true
  bUsedWithSkeletalMesh=true
  pixelDepthOffsetMode=PDOM_Legacy
  shadingModel=MSM_FromMaterialExpression
  usedShadingModels=Default Lit | Subsurface | Clear Coat | Subsurface Profile | Cloth
 }
 nodes [
  n1 = MaterialFunctionCall() {materialFunction=&MF_CharacterParent_2023_PreFX}
  n2 = GetMaterialAttributes(None=MaterialFunctionCall!(In=@n1.Result, InMask=@n1.Result, InFXMask=@n1.Result) {materialFunction=/Game/Global/MaterialFunctions/TechArt/MF_Ice_TechArt}.Result) {attributeGetTypes [
    #1
    #2
    #3
    #4
   ]}
  n3 = ScalarParameter() {defaultValue=0.5, group=Ice_Mist, parameterName=EM_OG_Min}
  n4 = ScalarParameter() {defaultValue=1, group=Ice_Mist, parameterName=EM_OG_Max}
  n5 = NamedRerouteUsage()
  n6 = Add(A=Multiply(A=@n2.Emissive Color, B=LinearInterpolate(A=@n3, B=@n4, Alpha=@n5) {constA=0.5}), B=NamedRerouteUsage())
  n7 = NamedRerouteUsage()
  n8 = Multiply(A=@n2.Base Color, B=ScalarParameter() {defaultValue=0.2, group=Ice_Mist, parameterName=BaseColorOverride})
  MaterialFunctionCall(
   Material=SetMaterialAttributes!(
    MaterialAttributes=@n2.MaterialAttributes
    Emissive Color=LinearInterpolate(A=@n2.Emissive Color, B=@n6, Alpha=@n7)
    Base Color=LinearInterpolate(A=@n2.Base Color, B=@n8, Alpha=@n7)
   ) {attributeSetTypes [#2, #3]}
   M=@n1.M
   Pre Skinned Local Position=@n1.M
   Distance Blend=@n1.M
   Fresnel=@n1.M
  ) {materialFunction=&MF_CharacterParent_2023_PostFX}
  NamedRerouteDeclaration(None=@n2.Normal) {name=Normal_IN, nodeColor=<rgba 1 0.1172 0 1>}
  n9 = StaticComponentMaskParameter(None=@n1.Tech Art Mask) {defaultR=true, group=Ice_Mist, parameterName=IceChannel}
  n10 = StaticComponentMaskParameter(None=@n1.Tech Art Mask) {defaultG=true, group=Ice_Mist, parameterName=OffHandChannel}
  NamedRerouteDeclaration(None=StaticSwitchParameter(True=Add(A=@n9, B=@n10), False=OneMinus(None=@n9)) {group=Ice_Mist, parameterName=AsPickaxe}) {name=Ice_MistMask, nodeColor=<rgba 0 1 0.2969 1>}
  NamedRerouteDeclaration(None=GetMaterialAttributes(None=@n1.Result) {attributeGetTypes [#3]}.Base Color) {name=BaseColorOG, nodeColor=<rgba 1 0.7969 0 1>}
  n11 = TextureCoordinate()
  n12 = ScalarParameter() {defaultValue=4, group=Ice_Mist, parameterName=MistNormalTiling01}
  n13 = TextureCoordinate()
  n14 = ScalarParameter() {defaultValue=8, group=Ice_Mist, parameterName=MistTiling02}
  n15 = ScalarParameter() {group=Ice_Mist, parameterName=MistDepth02}
  n16 = NamedRerouteUsage()
  n17 = TextureSample(UVs=BumpOffset(Coordinate=Multiply(A=@n13, B=@n14), Height=Multiply(A=@n15, B=@n16)), Tex=NamedRerouteUsage()) {
   samplerSource=SSM_Wrap_WorldGroupSettings
   samplerType=SAMPLERTYPE_Masks
   texture=/BRCosmetics/Accessories/FORT_Backpacks/Backpack_F_MED_PaddleHarp_Gap/Textures/T_Ice_Mist_Mask
  }
  n18 = LinearInterpolate(Alpha=@n17.B) {constA=0.7}
  n19 = ScalarParameter() {defaultValue=-10, group=Ice_Mist, parameterName=MistNormalDepth01}
  n20 = TextureSample(UVs=BumpOffset(Coordinate=Multiply(A=@n11, B=@n12), Height=Multiply(A=@n18, B=@n19))) {samplerType=SAMPLERTYPE_Normal, texture=/BRCosmetics/Accessories/FORT_Backpacks/Backpack_F_MED_PaddleHarp_Gap/Textures/T_Ice_Mist_normal}
  n21 = MaterialFunctionCall(Normal=@n20.RGB, Flatness=ScalarParameter() {group=Ice_Mist, parameterName=CracksFlatten}) {materialFunction=&FlattenNormal}
  n22 = ReflectionVectorWS(CustomWorldNormal=Transform(None=@n21.Result))
  n23 = NamedRerouteUsage()
  n24 = Power(Base=Saturate(None=DotProduct(A=@n22, B=@n23)), Exp=ScalarParameter() {defaultValue=2, group=Ice_Mist, parameterName=SpecPowerCracks}) {constExponent=3}
  n25 = ScalarParameter() {defaultValue=3, group=Ice_Mist, parameterName=SpecColorCracks}
  n26 = Saturate(None=Distance(A=@n21.Result, B=Constant3Vector() {constant=<rgba 0 0 1 0>}))
  n27 = ScalarParameter() {defaultValue=1, group=Ice_Mist, parameterName=SpecPowCracksBack}
  n28 = Add(
   A=Multiply(A=Multiply(A=@n24, B=@n25), B=@n26)
   B=Multiply(A=Power(Base=@n26, Exp=@n27), B=ScalarParameter() {defaultValue=0.4, group=Ice_Mist, parameterName=SpecColorCracksBack}) {constB=0.2}
  )
  NamedRerouteDeclaration(None=Saturate(None=@n28)) {name=Cracks, nodeColor=<rgba 1 0 0.1875 1>}
  NamedRerouteDeclaration(None=StaticSwitchParameter(True=@n10, False=Constant()) {group=Ice_Mist, parameterName=AsPickaxe}) {name=OffHandAlpha, nodeColor=<rgba 1 0.6328 0 1>}
  NamedRerouteDeclaration(None=StaticComponentMaskParameter(None=@n1.Tech Art Mask) {defaultB=true, group=Ice_Mist, parameterName=IceEdgeChannel}) {name=EdgeMask, nodeColor=<rgba 1 0 0.8437 1>}
  NamedRerouteDeclaration(None=@n2.Base Color) {name=BaseColorIN, nodeColor=<rgba 1 0 0.375 1>}
  n29 = LinearInterpolate(
   A=ScalarParameter() {defaultValue=1, group=Ice_Mist, parameterName=OverallEmissiveIntensity}
   B=ScalarParameter() {defaultValue=3, group=Ice_Mist, parameterName=OverallEmissiveIntensity_Offhand}
   Alpha=NamedRerouteUsage()
  )
  n30 = NamedRerouteUsage()
  n31 = ScalarParameter() {defaultValue=0.1, group=Ice_Mist, parameterName=SpecMin}
  n32 = ScalarParameter() {defaultValue=0.2, group=Ice_Mist, parameterName=SpecMax}
  n33 = NamedRerouteUsage()
  n34 = NamedRerouteUsage()
  n35 = ScalarParameter() {defaultValue=-1, group=Ice_Mist, parameterName=NormalFlatten}
  n36 = NamedRerouteDeclaration(None=Transform(None=MaterialFunctionCall(Normal=@n34, Flatness=@n35) {materialFunction=&FlattenNormal}.Result)) {name=Normal_WS, nodeColor=<rgba 0.5234 1 0 1>}
  n37 = ReflectionVectorWS(CustomWorldNormal=@n36)
  n38 = Power(Base=DotProduct(A=@n33, B=@n37), Exp=ScalarParameter() {defaultValue=5, group=Ice_Mist, parameterName=SpecPower})
  n39 = TextureSample() {samplerType=SAMPLERTYPE_Masks, texture=/BRCosmetics/Weapons/FORT_Melee/Pickaxe_PaddleHarp_Gap/Textures/T_paddlepickaxe_Pickaxe_FX_Add}
  n40 = StaticSwitchParameter(True=@n39.R, False=Constant()) {group=Ice_Mist, parameterName=AsPickaxe}
  n41 = Multiply(
   A=Multiply(
    A=SmoothStep(Min=@n31, Max=@n32, Value=@n38) {constMax=0.5, constMin=0.3}
    B=ScalarParameter() {defaultValue=1, group=Ice_Mist, parameterName=FakeSpecIntensity}
   ) {constB=20}
   B=OneMinus(None=NamedRerouteDeclaration(None=@n40) {name=FixMap, nodeColor=<rgba 1 0 0.8672 1>})
  )
  n42 = Multiply(A=@n30, B=VectorParameter() {group=Ice_Mist, parameterName=TransitionTint, defaultValue=<rgba 0 0.5617 1 1>}.RGB)
  n43 = DotProduct(A=Multiply(A=@n33) {constB=-1}, B=@n37)
  n44 = LinearInterpolate(A=ScalarParameter() {defaultValue=4, group=Ice_Mist, parameterName=TransitionPower}, Alpha=@n40)
  n45 = Multiply(
   A=LinearInterpolate(Alpha=NamedRerouteUsage()) {constA=0.5}
   B=Multiply(A=Power(Base=@n43, Exp=@n44) {constExponent=20}, B=ScalarParameter() {defaultValue=3, group=Ice_Mist, parameterName=FakeSSSIntensity}) {constB=20}
  )
  n46 = Add(A=Add(A=Multiply(A=@n30, B=@n41), B=Multiply(A=@n42, B=@n45)), B=NamedRerouteUsage())
  n47 = Constant()
  n48 = LinearInterpolate(
   B=ScalarParameter() {defaultValue=2.5, group=Ice_Mist, parameterName=MaxAudioIntensity}
   Alpha=Power(Base=MaterialFunctionCall!(InAmplitude, InSpectrum, CustomFakeSpectrum, CustomFakeAmplitude) {materialFunction=/BRCosmetics/Global/MaterialFunctions/AudioReactivity/MF_AudioReactivity_FrontendLoop}.Amplitude) {constExponent=4}
  ) {constB=3}
  n49 = MaterialFunctionCall(In=NamedRerouteUsage(), Contrast=Constant() {r=0.05}) {materialFunction=&CheapContrast}
  n50 = Multiply(
   A=Multiply(
    A=Multiply(A=@n48, B=@n49.Result)
    B=VectorParameter() {group=Ice_Mist, parameterName=ReactivityTint, defaultValue=<rgba 0.7052 0.8898 1 1>}.RGB
   )
   B=LinearInterpolate(Alpha=NamedRerouteUsage()) {constA=0.5, constB=2}
  )
  NamedRerouteDeclaration(None=Add(A=Multiply(A=@n29, B=@n46), B=StaticSwitchParameter(True=@n47, False=@n50) {group=Ice_Mist, parameterName=AsPickaxe})) {name=EM_OUT, nodeColor=<rgba 0 0.6953 1 1>}
  n51 = DotProduct(A=NamedRerouteUsage(), B=@n36)
  NamedRerouteDeclaration(None=SmoothStep(Value=Saturate(None=@n51)) {constMax=0.4, constMin=0.2}) {name=Lit, nodeColor=<rgba 1 0.2344 0 1>}
  n52 = TextureObjectParameter() {
   group=Ice_Mist
   parameterName=MistTexture
   samplerType=SAMPLERTYPE_Masks
   texture=&T_Ice_Mist_Mask
  }
  NamedRerouteDeclaration(None=@n52) {name=MistTexture, nodeColor=<rgba 0.3125 0 1 1>}
  n53 = TextureCoordinate()
  n54 = ScalarParameter() {defaultValue=8, group=Ice_Mist, parameterName=MistTiling01}
  n55 = ScalarParameter() {defaultValue=-2, group=Ice_Mist, parameterName=MistDepth01}
  n56 = NamedRerouteUsage()
  n57 = TextureSample(UVs=BumpOffset(Coordinate=Multiply(A=@n53, B=@n54), Height=Multiply(A=@n55, B=@n56)), Tex=NamedRerouteUsage()) {
   samplerSource=SSM_Wrap_WorldGroupSettings
   samplerType=SAMPLERTYPE_Masks
   texture=&T_Ice_Mist_Mask
  }
  n58 = Max(A=@n57.R, B=@n17.B)
  n59 = NamedRerouteUsage()
  n60 = Multiply(
   A=Multiply(A=NamedRerouteDeclaration(None=@n58) {name=MistGradient01, nodeColor=<rgba 0.6172 0 1 1>}, B=LinearInterpolate(Alpha=@n59) {constA=0.7})
   B=NamedRerouteUsage()
  )
  n61 = OneMinus(None=Fresnel() {baseReflectFraction=0, exponent=3})
  n62 = VectorParameter() {group=Ice_Mist, parameterName=EdgeTint, defaultValue=<rgba 0.7219 0.9528 1 1>}
  n63 = VectorParameter() {group=Ice_Mist, parameterName=EdgeTint_Offhand, defaultValue=<rgba 0 0.2764 1 3>}
  n64 = Multiply(A=@n63.RGB, B=@n63.RGB)
  n65 = NamedRerouteUsage()
  n66 = ScalarParameter() {defaultValue=0.1, parameterName=EdgeMin}
  n67 = ScalarParameter() {defaultValue=0.3, group=Ice_Mist, parameterName=EdgeMax}
  n68 = NamedRerouteUsage()
  n69 = LinearInterpolate(
   B=Multiply(A=ScalarParameter() {defaultValue=5, group=Ice_Mist, parameterName=EdgeIntensity}, B=LinearInterpolate(A=@n62.RGB, B=@n64, Alpha=@n65))
   Alpha=NamedRerouteDeclaration(None=SmoothStep(Min=@n66, Max=@n67, Value=@n68)) {name=MistEdge, nodeColor=<rgba 0 0.1094 1 1>}
  ) {constA=1, constB=3}
  n70 = Constant() {r=1}
  n71 = Add(A=@n28, B=ScalarParameter() {defaultValue=0.1, group=Ice_Mist, parameterName=BaseColorMist})
  n72 = NamedRerouteUsage()
  n73 = Multiply(
   A=Multiply(A=StaticSwitchParameter(True=@n69, False=@n70) {group=Ice_Mist, parameterName=AsPickaxe}, B=Multiply(A=@n71, B=@n72))
   B=OneMinus(None=NamedRerouteUsage())
  )
  n74 = LinearInterpolate(Alpha=NamedRerouteUsage()) {constA=0.5}
  NamedRerouteDeclaration(None=Add(A=Multiply(A=@n60, B=@n61), B=Multiply(A=@n73, B=@n74))) {name=Mist01, nodeColor=<rgba 0 1 0.2266 1>}
  NamedRerouteDeclaration(None=@n2.Roughness) {name=RoughnessIN, nodeColor=<rgba 1 0 0.2109 1>}
  NamedRerouteDeclaration(None=Custom!(None) {code="return View.DirectionalLightDirection;", description=DirectionalLightDirection, showCode=true}) {name=LightVector_IN, nodeColor=<rgba 0 0.4141 1 1>}
  ScalarParameter() {defaultValue=1, group=Ice_Mist, parameterName=EnabbleMusicReactivity}
 ]
}
