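# Animation Blueprint for the PaddleHarp Gap backpack cosmetic that adds procedural bobbing and spring motion to its moon and gap bones.
# A time-based sine drives base_moon, then root-relative, zero-gravity Anim Dynamics springs dyn_moon and six gap bones, amid fixed additive offsets and zero-offset shard transforms.
# Updates also compute staggered shard sines, noisy speaker vibrations and music amplitude from a Material Parameter Collection, none visibly wired into the pose.
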AnimBlueprint /BRCosmetics/Accessories/FORT_Backpacks/Backpack_F_MED_PaddleHarp_Gap/Meshes/F_MED_PaddleHarp_Gap_Pack_AnimBP
target_skeleton=&/F_MED_PaddleHarp_Gap_Pack_Skeleton
graphs [
 AnimGraph {nodes [
   n1 = "Transform (Modify) Bone - Bone: base_moon"(
    ComponentPose=Local To Component(LocalPose=Use cached pose 'Input_pose_cache'())
    Alpha=0.1
    Translation="Get Translate Shard 1 "() -> ("Translate Shard 1 ":Vector)
    Rotation="0,0,0"
    Scale="1,1,1"
   ) {
    rotationMode=BMM_Additive
    scaleSpace=BCS_BoneSpace
    translationMode=BMM_Additive
    translationSpace=BCS_BoneSpace
    alphaBoolBlend {blendOption=Linear}
    boneToModify {boneName=base_moon}
   }
   n2 = Anim Dynamics - Bone: dyn_moon(ComponentPose=@n1, Alpha=0.5, ExternalForce="0,0,0") {
    angularDampingOverride=0.9
    angularSpringConstant=200
    bGravityOverrideInSimSpace=true
    bLinearSpring=true
    bOverrideAngularDamping=true
    bOverrideLinearDamping=true
    gravityScale=0
    linearDampingOverride=0.9999
    linearSpringConstant=100
    simulationSpace=RootRelative
    alphaBoolBlend {blendOption=Linear}
    boundBone {boneName=dyn_moon}
    componentAppliedLinearAccClamp=<50000 50000 50000>
    componentLinearAccScale {x=0.1, z=0.1}
    componentLinearVelScale=<0.01 0.01 0.01>
    physicsBodyDefinitions [{
      sphereCollisionRadius=0
      boundBone {boneName=dyn_moon}
      boxExtents=<4 4 4>
      constraintSetup {angularLimitsMax=<2 2 2>, angularLimitsMin=<-2 -2 -2>, angularTarget=<0 0 0>, linearAxesMax=<4 4 4>, linearAxesMin=<-1 -1 -1>}
      localJointOffset=<0 0 0>
     }]
    relativeSpaceBone {boneName=root}
   }
   n3 = "Transform (Modify) Bone - Bone: dyn_paddleHarp_gap_01"!(ComponentPose=Local To Component(LocalPose=Component To Local(ComponentPose=@n2)), Alpha=1, Translation="1,0,0") -> (Pose) {
    rotationMode=BMM_Additive
    scaleSpace=BCS_BoneSpace
    translationMode=BMM_Additive
    translationSpace=BCS_BoneSpace
    alphaBoolBlend {blendOption=Linear}
    boneToModify {boneName=dyn_paddleHarp_gap_01}
    translation {x=1}
   }
   n4 = "Transform (Modify) Bone - Bone: dyn_paddleHarp_gap_02"!(ComponentPose=@n3, Alpha=1, Translation="1,0,0") -> (Pose) {
    rotationMode=BMM_Additive
    scaleSpace=BCS_BoneSpace
    translationMode=BMM_Additive
    translationSpace=BCS_BoneSpace
    alphaBoolBlend {blendOption=Linear}
    boneToModify {boneName=dyn_paddleHarp_gap_02}
    translation {x=1}
   }
   n5 = "Transform (Modify) Bone - Bone: dyn_paddleHarp_gap_03"!(ComponentPose=@n4, Alpha=1, Translation="1,0,0") -> (Pose) {
    rotationMode=BMM_Additive
    scaleSpace=BCS_BoneSpace
    translationMode=BMM_Additive
    translationSpace=BCS_BoneSpace
    alphaBoolBlend {blendOption=Linear}
    boneToModify {boneName=dyn_paddleHarp_gap_03}
    translation {x=1}
   }
   n6 = "Transform (Modify) Bone - Bone: dyn_paddleHarp_gap_04"!(ComponentPose=@n5, Alpha=1, Translation="1,0,0") -> (Pose) {
    rotationMode=BMM_Additive
    scaleSpace=BCS_BoneSpace
    translationMode=BMM_Additive
    translationSpace=BCS_BoneSpace
    alphaBoolBlend {blendOption=Linear}
    boneToModify {boneName=dyn_paddleHarp_gap_04}
    translation {x=1}
   }
   n7 = "Transform (Modify) Bone - Bone: dyn_paddleHarp_gap_05"!(ComponentPose=@n6, Alpha=1, Translation="1,0,0") -> (Pose) {
    rotationMode=BMM_Additive
    scaleSpace=BCS_BoneSpace
    translationMode=BMM_Additive
    translationSpace=BCS_BoneSpace
    alphaBoolBlend {blendOption=Linear}
    boneToModify {boneName=dyn_paddleHarp_gap_05}
    translation {x=1}
   }
   n8 = "Transform (Modify) Bone - Bone: dyn_paddleHarp_gap_06"!(ComponentPose=@n7, Alpha=1, Translation="2,0,0") -> (Pose) {
    rotationMode=BMM_Additive
    scaleSpace=BCS_BoneSpace
    translationMode=BMM_Additive
    translationSpace=BCS_BoneSpace
    alphaBoolBlend {blendOption=Linear}
    boneToModify {boneName=dyn_paddleHarp_gap_06}
    translation {x=2}
   }
   n9 = "Transform (Modify) Bone - Bone: dyn_paddleHarp_gap_01"!(ComponentPose=@n8, Alpha=1, Translation="0,0,0", Rotation:Rotator="0,0,0", Scale:Vector="1,1,1") -> (Pose) {
    rotationMode=BMM_Additive
    scaleSpace=BCS_BoneSpace
    translationMode=BMM_Additive
    translationSpace=BCS_BoneSpace
    alphaBoolBlend {blendOption=Linear}
    boneToModify {boneName=dyn_paddleHarp_gap_01}
   }
   n10 = "Transform (Modify) Bone - Bone: dyn_paddleHarp_gap_02"!(ComponentPose=@n9, Alpha=1, Translation="0,0,0", Rotation:Rotator="0,0,0", Scale:Vector="1,1,1") -> (Pose) {
    rotationMode=BMM_Additive
    scaleSpace=BCS_BoneSpace
    translationMode=BMM_Additive
    translationSpace=BCS_BoneSpace
    alphaBoolBlend {blendOption=Linear}
    boneToModify {boneName=dyn_paddleHarp_gap_02}
   }
   n11 = "Transform (Modify) Bone - Bone: dyn_paddleHarp_gap_03"!(ComponentPose=@n10, Alpha=1, Translation="0,0,0", Rotation:Rotator="0,0,0", Scale:Vector="1,1,1") -> (Pose) {
    rotationMode=BMM_Additive
    scaleSpace=BCS_BoneSpace
    translationMode=BMM_Additive
    translationSpace=BCS_BoneSpace
    alphaBoolBlend {blendOption=Linear}
    boneToModify {boneName=dyn_paddleHarp_gap_03}
   }
   n12 = "Transform (Modify) Bone - Bone: dyn_paddleHarp_gap_04"!(ComponentPose=@n11, Alpha=1, Translation="0,0,0", Rotation:Rotator="0,0,0", Scale:Vector="1,1,1") -> (Pose) {
    rotationMode=BMM_Additive
    scaleSpace=BCS_BoneSpace
    translationMode=BMM_Additive
    translationSpace=BCS_BoneSpace
    alphaBoolBlend {blendOption=Linear}
    boneToModify {boneName=dyn_paddleHarp_gap_04}
   }
   n13 = "Transform (Modify) Bone - Bone: dyn_paddleHarp_gap_05"!(ComponentPose=@n12, Alpha=1, Translation="0,0,0", Rotation:Rotator="0,0,0", Scale:Vector="1,1,1") -> (Pose) {
    rotationMode=BMM_Additive
    scaleSpace=BCS_BoneSpace
    translationMode=BMM_Additive
    translationSpace=BCS_BoneSpace
    alphaBoolBlend {blendOption=Linear}
    boneToModify {boneName=dyn_paddleHarp_gap_05}
   }
   n14 = "Transform (Modify) Bone - Bone: dyn_paddleHarp_gap_06"!(ComponentPose=@n13, Alpha=1, Translation="0,0,0", Rotation:Rotator="0,0,0", Scale:Vector="1,1,1") -> (Pose) {
    rotationMode=BMM_Additive
    scaleSpace=BCS_BoneSpace
    translationMode=BMM_Additive
    translationSpace=BCS_BoneSpace
    alphaBoolBlend {blendOption=Linear}
    boneToModify {boneName=dyn_paddleHarp_gap_06}
   }
   n15 = Anim Dynamics - Bone: dyn_paddleHarp_gap_01(ComponentPose=@n14, Alpha=0.5, ExternalForce="0,0,0") {
    angularDampingOverride=0.9
    angularSpringConstant=200
    bGravityOverrideInSimSpace=true
    bLinearSpring=true
    bOverrideAngularDamping=true
    bOverrideLinearDamping=true
    gravityScale=0
    linearDampingOverride=0.9999
    linearSpringConstant=100
    simulationSpace=RootRelative
    alphaBoolBlend {blendOption=Linear}
    boundBone {boneName=dyn_paddleHarp_gap_01}
    componentAppliedLinearAccClamp=<50000 50000 50000>
    componentLinearAccScale {x=0.1, z=0.1}
    componentLinearVelScale=<0.01 0.01 0.01>
    physicsBodyDefinitions [{
      sphereCollisionRadius=0
      boundBone {boneName=dyn_paddleHarp_gap_01}
      boxExtents=<4 4 4>
      constraintSetup {angularLimitsMax=<2 2 2>, angularLimitsMin=<-2 -2 -2>, angularTarget=<0 0 0>, linearAxesMax=<4 4 4>, linearAxesMin=<-1 -1 -1>}
      localJointOffset=<0 0 0>
     }]
    relativeSpaceBone {boneName=root}
   }
   n16 = Anim Dynamics - Bone: dyn_paddleHarp_gap_02(ComponentPose=@n15, Alpha=0.5, ExternalForce="0,0,0") {
    angularDampingOverride=0.9
    angularSpringConstant=200
    bGravityOverrideInSimSpace=true
    bLinearSpring=true
    bOverrideAngularDamping=true
    bOverrideLinearDamping=true
    gravityScale=0
    linearDampingOverride=0.9999
    linearSpringConstant=100
    simulationSpace=RootRelative
    alphaBoolBlend {blendOption=Linear}
    boundBone {boneName=dyn_paddleHarp_gap_02}
    componentAppliedLinearAccClamp=<50000 50000 50000>
    componentLinearAccScale {x=0.1, z=0.1}
    componentLinearVelScale=<0.01 0.01 0.01>
    physicsBodyDefinitions [{
      sphereCollisionRadius=0
      boundBone {boneName=dyn_paddleHarp_gap_02}
      boxExtents=<4 4 4>
      constraintSetup {angularLimitsMax=<2 2 2>, angularLimitsMin=<-2 -2 -2>, angularTarget=<0 0 0>, linearAxesMax=<4 4 4>, linearAxesMin=<-1 -1 -1>}
      localJointOffset=<0 0 0>
     }]
    relativeSpaceBone {boneName=root}
   }
   n17 = Anim Dynamics - Bone: dyn_paddleHarp_gap_03(ComponentPose=@n16, Alpha=0.5, ExternalForce="0,0,0") {
    angularDampingOverride=0.9
    angularSpringConstant=200
    bGravityOverrideInSimSpace=true
    bLinearSpring=true
    bOverrideAngularDamping=true
    bOverrideLinearDamping=true
    gravityScale=0
    linearDampingOverride=0.9999
    linearSpringConstant=100
    simulationSpace=RootRelative
    alphaBoolBlend {blendOption=Linear}
    boundBone {boneName=dyn_paddleHarp_gap_03}
    componentAppliedLinearAccClamp=<50000 50000 50000>
    componentLinearAccScale {x=0.1, z=0.1}
    componentLinearVelScale=<0.01 0.01 0.01>
    physicsBodyDefinitions [{
      sphereCollisionRadius=0
      boundBone {boneName=dyn_paddleHarp_gap_03}
      boxExtents=<4 4 4>
      constraintSetup {angularLimitsMax=<2 2 2>, angularLimitsMin=<-2 -2 -2>, angularTarget=<0 0 0>, linearAxesMax=<4 4 4>, linearAxesMin=<-1 -1 -1>}
      localJointOffset=<0 0 0>
     }]
    relativeSpaceBone {boneName=root}
   }
   n18 = Anim Dynamics - Bone: dyn_paddleHarp_gap_04(ComponentPose=@n17, Alpha=0.5, ExternalForce="0,0,0") {
    angularDampingOverride=0.9
    angularSpringConstant=200
    bGravityOverrideInSimSpace=true
    bLinearSpring=true
    bOverrideAngularDamping=true
    bOverrideLinearDamping=true
    gravityScale=0
    linearDampingOverride=0.9999
    linearSpringConstant=100
    simulationSpace=RootRelative
    alphaBoolBlend {blendOption=Linear}
    boundBone {boneName=dyn_paddleHarp_gap_04}
    componentAppliedLinearAccClamp=<50000 50000 50000>
    componentLinearAccScale {x=0.1, z=0.1}
    componentLinearVelScale=<0.01 0.01 0.01>
    physicsBodyDefinitions [{
      sphereCollisionRadius=0
      boundBone {boneName=dyn_paddleHarp_gap_04}
      boxExtents=<4 4 4>
      constraintSetup {angularLimitsMax=<2 2 2>, angularLimitsMin=<-2 -2 -2>, angularTarget=<0 0 0>, linearAxesMax=<4 4 4>, linearAxesMin=<-1 -1 -1>}
      localJointOffset=<0 0 0>
     }]
    relativeSpaceBone {boneName=root}
   }
   n19 = Anim Dynamics - Bone: dyn_paddleHarp_gap_05(ComponentPose=@n18, Alpha=0.5, ExternalForce="0,0,0") {
    angularDampingOverride=0.9
    angularSpringConstant=200
    bGravityOverrideInSimSpace=true
    bLinearSpring=true
    bOverrideAngularDamping=true
    bOverrideLinearDamping=true
    gravityScale=0
    linearDampingOverride=0.9999
    linearSpringConstant=100
    simulationSpace=RootRelative
    alphaBoolBlend {blendOption=Linear}
    boundBone {boneName=dyn_paddleHarp_gap_05}
    componentAppliedLinearAccClamp=<50000 50000 50000>
    componentLinearAccScale {x=0.1, z=0.1}
    componentLinearVelScale=<0.01 0.01 0.01>
    physicsBodyDefinitions [{
      sphereCollisionRadius=0
      boundBone {boneName=dyn_paddleHarp_gap_05}
      boxExtents=<4 4 4>
      constraintSetup {angularLimitsMax=<2 2 2>, angularLimitsMin=<-2 -2 -2>, angularTarget=<0 0 0>, linearAxesMax=<4 4 4>, linearAxesMin=<-1 -1 -1>}
      localJointOffset=<0 0 0>
     }]
    relativeSpaceBone {boneName=root}
   }
   n20 = Anim Dynamics - Bone: dyn_paddleHarp_gap_06(ComponentPose=@n19, Alpha=0.5, ExternalForce="0,0,0") {
    angularDampingOverride=0.9
    angularSpringConstant=200
    bGravityOverrideInSimSpace=true
    bLinearSpring=true
    bOverrideAngularDamping=true
    bOverrideLinearDamping=true
    gravityScale=0
    linearDampingOverride=0.9999
    linearSpringConstant=100
    simulationSpace=RootRelative
    alphaBoolBlend {blendOption=Linear}
    boundBone {boneName=dyn_paddleHarp_gap_06}
    componentAppliedLinearAccClamp=<50000 50000 50000>
    componentLinearAccScale {x=0.1, z=0.1}
    componentLinearVelScale=<0.01 0.01 0.01>
    physicsBodyDefinitions [{
      sphereCollisionRadius=0
      boundBone {boneName=dyn_paddleHarp_gap_06}
      boxExtents=<4 4 4>
      constraintSetup {angularLimitsMax=<2 2 2>, angularLimitsMin=<-2 -2 -2>, angularTarget=<0 0 0>, linearAxesMax=<4 4 4>, linearAxesMin=<-1 -1 -1>}
      localJointOffset=<0 0 0>
     }]
    relativeSpaceBone {boneName=root}
   }
   n21 = "Transform (Modify) Bone - Bone: dyn_shards_1"(
    ComponentPose=Local To Component(LocalPose=Component To Local(ComponentPose=@n20))
    Alpha=1
    Translation="0,0,0"
    Rotation="0,0,0"
    Scale="1,1,1"
   ) {
    rotationMode=BMM_Additive
    scaleSpace=BCS_BoneSpace
    translationMode=BMM_Additive
    translationSpace=BCS_BoneSpace
    alphaBoolBlend {blendOption=Linear}
    boneToModify {boneName=dyn_shards_1}
   }
   n22 = "Transform (Modify) Bone - Bone: dyn_shards_2"(ComponentPose=@n21, Alpha=1, Translation="0,0,0", Rotation="0,0,0", Scale="1,1,1") {
    rotationMode=BMM_Additive
    scaleSpace=BCS_BoneSpace
    translationMode=BMM_Additive
    translationSpace=BCS_BoneSpace
    alphaBoolBlend {blendOption=Linear}
    boneToModify {boneName=dyn_shards_2}
   }
   n23 = "Transform (Modify) Bone - Bone: dyn_paddleHarp_gap_03"(ComponentPose=@n22, Alpha=1, Translation="0,0,0", Rotation="0,0,0", Scale="1,1,1") {
    rotationMode=BMM_Additive
    scaleSpace=BCS_BoneSpace
    translationMode=BMM_Additive
    translationSpace=BCS_BoneSpace
    alphaBoolBlend {blendOption=Linear}
    boneToModify {boneName=dyn_paddleHarp_gap_03}
   }
   n24 = "Transform (Modify) Bone - Bone: dyn_shards_4"(ComponentPose=@n23, Alpha=1, Translation="0,0,0", Rotation="0,0,0", Scale="1,1,1") {
    rotationMode=BMM_Additive
    scaleSpace=BCS_BoneSpace
    translationMode=BMM_Additive
    translationSpace=BCS_BoneSpace
    alphaBoolBlend {blendOption=Linear}
    boneToModify {boneName=dyn_shards_4}
   }
   n25 = "Transform (Modify) Bone - Bone: dyn_shards_5"(ComponentPose=@n24, Alpha=1, Translation="0,0,0", Rotation="0,0,0", Scale="1,1,1") {
    rotationMode=BMM_Additive
    scaleSpace=BCS_BoneSpace
    translationMode=BMM_Additive
    translationSpace=BCS_BoneSpace
    alphaBoolBlend {blendOption=Linear}
    boneToModify {boneName=dyn_shards_5}
   }
   n26 = "Transform (Modify) Bone - Bone: dyn_shards_6"(ComponentPose=@n25, Alpha=1, Translation="0,0,0", Rotation="0,0,0", Scale="1,1,1") {
    rotationMode=BMM_Additive
    scaleSpace=BCS_BoneSpace
    translationMode=BMM_Additive
    translationSpace=BCS_BoneSpace
    alphaBoolBlend {blendOption=Linear}
    boneToModify {boneName=dyn_shards_6}
   }
   n27 = "Transform (Modify) Bone - Bone: dyn_shards_7"(ComponentPose=@n26, Alpha=1, Translation="0,0,0", Rotation="0,0,0", Scale="1,1,1") {
    rotationMode=BMM_Additive
    scaleSpace=BCS_BoneSpace
    translationMode=BMM_Additive
    translationSpace=BCS_BoneSpace
    alphaBoolBlend {blendOption=Linear}
    boneToModify {boneName=dyn_shards_7}
   }
   n28 = "Transform (Modify) Bone - Bone: dyn_shards_8"(ComponentPose=@n27, Alpha=1, Translation="0,0,0", Rotation="0,0,0", Scale="1,1,1") {
    rotationMode=BMM_Additive
    scaleSpace=BCS_BoneSpace
    translationMode=BMM_Additive
    translationSpace=BCS_BoneSpace
    alphaBoolBlend {blendOption=Linear}
    boneToModify {boneName=dyn_shards_8}
   }
   n29 = "Transform (Modify) Bone - Bone: dyn_shards_9"(ComponentPose=@n28, Alpha=1, Translation="0,0,0", Rotation="0,0,0", Scale="1,1,1") {
    rotationMode=BMM_Additive
    scaleSpace=BCS_BoneSpace
    translationMode=BMM_Additive
    translationSpace=BCS_BoneSpace
    alphaBoolBlend {blendOption=Linear}
    boneToModify {boneName=dyn_shards_9}
   }
   n30 = "Transform (Modify) Bone - Bone: dyn_shards_10"(ComponentPose=@n29, Alpha=1, Translation="0,0,0", Rotation="0,0,0", Scale="1,1,1") {
    rotationMode=BMM_Additive
    scaleSpace=BCS_BoneSpace
    translationMode=BMM_Additive
    translationSpace=BCS_BoneSpace
    alphaBoolBlend {blendOption=Linear}
    boneToModify {boneName=dyn_shards_10}
   }
   n31 = "Transform (Modify) Bone - Bone: dyn_shards_11"(ComponentPose=@n30, Alpha=1, Translation="0,0,0", Rotation="0,0,0", Scale="1,1,1") {
    rotationMode=BMM_Additive
    scaleSpace=BCS_BoneSpace
    translationMode=BMM_Additive
    translationSpace=BCS_BoneSpace
    alphaBoolBlend {blendOption=Linear}
    boneToModify {boneName=dyn_shards_11}
   }
   n32 = "Transform (Modify) Bone - Bone: dyn_shards_12"(ComponentPose=@n31, Alpha=1, Translation="0,0,0", Rotation="0,0,0", Scale="1,1,1") {
    rotationMode=BMM_Additive
    scaleSpace=BCS_BoneSpace
    translationMode=BMM_Additive
    translationSpace=BCS_BoneSpace
    alphaBoolBlend {blendOption=Linear}
    boneToModify {boneName=dyn_shards_12}
   }
   n33 = "Transform (Modify) Bone - Bone: dyn_shards_13"(ComponentPose=@n32, Alpha=1, Translation="0,0,0", Rotation="0,0,0", Scale="1,1,1") {
    rotationMode=BMM_Additive
    scaleSpace=BCS_BoneSpace
    translationMode=BMM_Additive
    translationSpace=BCS_BoneSpace
    alphaBoolBlend {blendOption=Linear}
    boneToModify {boneName=dyn_shards_13}
   }
   n34 = "Transform (Modify) Bone - Bone: dyn_shards_14"(ComponentPose=@n33, Alpha=1, Translation="0,0,0", Rotation="0,0,0", Scale="1,1,1") {
    rotationMode=BMM_Additive
    scaleSpace=BCS_BoneSpace
    translationMode=BMM_Additive
    translationSpace=BCS_BoneSpace
    alphaBoolBlend {blendOption=Linear}
    boneToModify {boneName=dyn_shards_14}
   }
   n35 = "Transform (Modify) Bone - Bone: dyn_shards_15"(ComponentPose=@n34, Alpha=1, Translation="0,0,0", Rotation="0,0,0", Scale="1,1,1") {
    rotationMode=BMM_Additive
    scaleSpace=BCS_BoneSpace
    translationMode=BMM_Additive
    translationSpace=BCS_BoneSpace
    alphaBoolBlend {blendOption=Linear}
    boneToModify {boneName=dyn_shards_15}
   }
   n36 = "Transform (Modify) Bone - Bone: dyn_shards_16"(ComponentPose=@n35, Alpha=1, Translation="0,0,0", Rotation="0,0,0", Scale="1,1,1") {
    rotationMode=BMM_Additive
    scaleSpace=BCS_BoneSpace
    translationMode=BMM_Additive
    translationSpace=BCS_BoneSpace
    alphaBoolBlend {blendOption=Linear}
    boneToModify {boneName=dyn_shards_16}
   }
   n37 = "Transform (Modify) Bone - Bone: dyn_shards_17"(ComponentPose=@n36, Alpha=1, Translation="0,0,0", Rotation="0,0,0", Scale="1,1,1") {
    rotationMode=BMM_Additive
    scaleSpace=BCS_BoneSpace
    translationMode=BMM_Additive
    translationSpace=BCS_BoneSpace
    alphaBoolBlend {blendOption=Linear}
    boneToModify {boneName=dyn_shards_17}
   }
   n38 = "Transform (Modify) Bone - Bone: dyn_shards_18"(ComponentPose=@n37, Alpha=1, Translation="0,0,0", Rotation="0,0,0", Scale="1,1,1") {
    rotationMode=BMM_Additive
    scaleSpace=BCS_BoneSpace
    translationMode=BMM_Additive
    translationSpace=BCS_BoneSpace
    alphaBoolBlend {blendOption=Linear}
    boneToModify {boneName=dyn_shards_18}
   }
   n39 = "Transform (Modify) Bone - Bone: dyn_shards_19"(ComponentPose=@n38, Alpha=1, Translation="0,0,0", Rotation="0,0,0", Scale="1,1,1") {
    rotationMode=BMM_Additive
    scaleSpace=BCS_BoneSpace
    translationMode=BMM_Additive
    translationSpace=BCS_BoneSpace
    alphaBoolBlend {blendOption=Linear}
    boneToModify {boneName=dyn_shards_19}
   }
   n40 = "Transform (Modify) Bone - Bone: dyn_shards_20"(ComponentPose=@n39, Alpha=1, Translation="0,0,0", Rotation="0,0,0", Scale="1,1,1") {
    rotationMode=BMM_Additive
    scaleSpace=BCS_BoneSpace
    translationMode=BMM_Additive
    translationSpace=BCS_BoneSpace
    alphaBoolBlend {blendOption=Linear}
    boneToModify {boneName=dyn_shards_20}
   }
   n41 = "Transform (Modify) Bone - Bone: dyn_shards_21"(ComponentPose=@n40, Alpha=1, Translation="0,0,0", Rotation="0,0,0", Scale="1,1,1") {
    rotationMode=BMM_Additive
    scaleSpace=BCS_BoneSpace
    translationMode=BMM_Additive
    translationSpace=BCS_BoneSpace
    alphaBoolBlend {blendOption=Linear}
    boneToModify {boneName=dyn_shards_21}
   }
   n42 = "Transform (Modify) Bone - Bone: dyn_shards_22"(ComponentPose=@n41, Alpha=1, Translation="0,0,0", Rotation="0,0,0", Scale="1,1,1") {
    rotationMode=BMM_Additive
    scaleSpace=BCS_BoneSpace
    translationMode=BMM_Additive
    translationSpace=BCS_BoneSpace
    alphaBoolBlend {blendOption=Linear}
    boneToModify {boneName=dyn_shards_22}
   }
   n43 = "Transform (Modify) Bone - Bone: dyn_shards_23"(ComponentPose=@n42, Alpha=1, Translation="0,0,0", Rotation="0,0,0", Scale="1,1,1") {
    rotationMode=BMM_Additive
    scaleSpace=BCS_BoneSpace
    translationMode=BMM_Additive
    translationSpace=BCS_BoneSpace
    alphaBoolBlend {blendOption=Linear}
    boneToModify {boneName=dyn_shards_23}
   }
   n44 = "Transform (Modify) Bone - Bone: dyn_paddleHarp_gap_01"(
    ComponentPose=Local To Component(LocalPose=Component To Local(ComponentPose=@n43))
    Alpha=1
    Translation="5,0,0"
    Rotation="0,0,0"
    Scale="1,1,1"
   ) {
    rotationMode=BMM_Additive
    scaleSpace=BCS_BoneSpace
    translationMode=BMM_Additive
    translationSpace=BCS_BoneSpace
    alphaBoolBlend {blendOption=Linear}
    boneToModify {boneName=dyn_paddleHarp_gap_01}
    translation {x=5}
   }
   n45 = "Transform (Modify) Bone - Bone: dyn_paddleHarp_gap_02"(ComponentPose=@n44, Alpha=1, Translation="5,0,0", Rotation="0,0,0", Scale="1,1,1") {
    rotationMode=BMM_Additive
    scaleSpace=BCS_BoneSpace
    translationMode=BMM_Additive
    translationSpace=BCS_BoneSpace
    alphaBoolBlend {blendOption=Linear}
    boneToModify {boneName=dyn_paddleHarp_gap_02}
    translation {x=5}
   }
   n46 = "Transform (Modify) Bone - Bone: dyn_paddleHarp_gap_03"(ComponentPose=@n45, Alpha=1, Translation="5,0,0", Rotation="0,0,0", Scale="1,1,1") {
    rotationMode=BMM_Additive
    scaleSpace=BCS_BoneSpace
    translationMode=BMM_Additive
    translationSpace=BCS_BoneSpace
    alphaBoolBlend {blendOption=Linear}
    boneToModify {boneName=dyn_paddleHarp_gap_03}
    translation {x=5}
   }
   n47 = "Transform (Modify) Bone - Bone: dyn_paddleHarp_gap_04"(ComponentPose=@n46, Alpha=1, Translation="5,0,0", Rotation="0,0,0", Scale="1,1,1") {
    rotationMode=BMM_Additive
    scaleSpace=BCS_BoneSpace
    translationMode=BMM_Additive
    translationSpace=BCS_BoneSpace
    alphaBoolBlend {blendOption=Linear}
    boneToModify {boneName=dyn_paddleHarp_gap_04}
    translation {x=5}
   }
   n48 = "Transform (Modify) Bone - Bone: dyn_paddleHarp_gap_05"(ComponentPose=@n47, Alpha=1, Translation="5,0,0", Rotation="0,0,0", Scale="1,1,1") {
    rotationMode=BMM_Additive
    scaleSpace=BCS_BoneSpace
    translationMode=BMM_Additive
    translationSpace=BCS_BoneSpace
    alphaBoolBlend {blendOption=Linear}
    boneToModify {boneName=dyn_paddleHarp_gap_05}
    translation {x=5}
   }
   n49 = "Transform (Modify) Bone - Bone: dyn_paddleHarp_gap_06"(ComponentPose=@n48, Alpha=1, Translation="5,0,0", Rotation="0,0,0", Scale="1,1,1") {
    rotationMode=BMM_Additive
    scaleSpace=BCS_BoneSpace
    translationMode=BMM_Additive
    translationSpace=BCS_BoneSpace
    alphaBoolBlend {blendOption=Linear}
    boneToModify {boneName=dyn_paddleHarp_gap_06}
    translation {x=5}
   }
   Output Pose(Result=Component To Local(ComponentPose=@n49)) {layerGroup=None}
   Save cached pose 'Input_pose_cache'(Pose=Input Pose()) {cacheName=Input_pose_cache}
  ]}
 CalculateVIbration {nodes [
   n1 = Inputs!() -> (Frequency:double=~, Amplitude:double=~, NoiseAmplitude:double=~)
   n2 = RandomFloatInRange(Min=float * float(A=@n1.NoiseAmplitude, B=-1) {is=title K2Node_PromotableOperator}, Max=@n1.NoiseAmplitude)
   n3 = GetGameTimeInSeconds()
   n4 = float * float(A="Sin (Radians)"(A=float * float(A=@n1, B=@n3)), B=@n1.Amplitude)
   Outputs!(Result:double="Clamp (Float)"(Value=float + float(A=@n4, B=@n2), Min=-1, Max=10))
  ]}
 EventGraph {nodes [
   Event BlueprintUpdateAnimation()
   > n1 = Set Range(Range:double=1) -> (Output_Get:double)
   > Set alpha(alpha:double=0.75) -> (Output_Get:double)
   n2 = GetRealTimeSeconds()
   n3 = "Sin (Radians)"(A=float * float(A=float - float(A=@n2, B=0.6), B=5) {is=title K2Node_CommutativeAssociativeBinaryOperator})
   > "Set Translate Shard 1 "("Translate Shard 1 ":Vector=MakeVector(X=float / float(A=@n3, B=@n1) {is=title K2Node_CallFunction}, Y=0)) -> (Output_Get:Vector)
   n4 = GetRealTimeSeconds()
   n5 = "Sin (Radians)"(A=float * float(A=float - float(A=@n4, B=0.2), B=5))
   > Set Translate Shard 2(Translate Shard 2:Vector=MakeVector(X=float / float(A=@n5, B=@n1), Y=0)) -> (Output_Get:Vector)
   n6 = GetRealTimeSeconds()
   n7 = "Sin (Radians)"(A=float * float(A=float - float(A=@n6, B=0.4), B=5))
   > Set Translate Shard 3(Translate Shard 3:Vector=MakeVector(X=float / float(A=@n7, B=@n1), Y=0)) -> (Output_Get:Vector)
   n8 = GetRealTimeSeconds()
   n9 = "Sin (Radians)"(A=float * float(A=float - float(A=@n8, B=0.6), B=5))
   > Set Translate Shard 4(Translate Shard 4:Vector=MakeVector(X=float / float(A=@n9, B=@n1), Y=0)) -> (Output_Get:Vector)
   n10 = GetRealTimeSeconds()
   n11 = "Sin (Radians)"(A=float * float(A=float - float(A=@n10, B=0.8), B=5))
   > Set Translate Shard 5(Translate Shard 5:Vector=MakeVector(X=float / float(A=@n11, B=@n1), Y=0)) -> (Output_Get:Vector)
   n12 = GetRealTimeSeconds()
   n13 = "Sin (Radians)"(A=float * float(A=float - float(A=@n12, B=1), B=5))
   > Set Translate Shard 6(Translate Shard 6:Vector=MakeVector(X=float / float(A=@n13, B=@n1), Y=0)) -> (Output_Get:Vector)
   > Set Speakers_01(Speakers_01:Vector=MakeVector(
     X=Calculate VIbration!(Frequency:double=20, Amplitude:double=3.5, NoiseAmplitude:double=0.1) -> (Result:double=~) {is=title K2Node_MacroInstance}
     Y=0
    )) -> (Output_Get:Vector)
   > Set Speakers_02(Speakers_02:Vector=MakeVector(X=Calculate VIbration(Frequency=19, Amplitude=3, NoiseAmplitude=0.1) -> (Result=~), Y=0)) -> (Output_Get:Vector)
   > Set Speakers_03(Speakers_03:Vector=MakeVector(X=Calculate VIbration(Frequency=18, Amplitude=3.5, NoiseAmplitude=0.1) -> (Result=~), Y=0)) -> (Output_Get:Vector)
   > Set Speakers_04(Speakers_04:Vector=MakeVector(X=Calculate VIbration(Frequency=17, Amplitude=3, NoiseAmplitude=0.1) -> (Result=~), Y=0)) -> (Output_Get:Vector)
   > Set Speakers_05(Speakers_05:Vector=MakeVector(X=Calculate VIbration(Frequency=20, Amplitude=3, NoiseAmplitude=0.1) -> (Result=~), Y=0)) -> (Output_Get:Vector)
   > Set Speakers_06(Speakers_06:Vector=MakeVector(X=Calculate VIbration(Frequency=18, Amplitude=3, NoiseAmplitude=0.1) -> (Result=~), Y=0)) -> (Output_Get:Vector)
   > n14 = GetScalarParameterValue!(
    execute
    Collection:Material Parameter Collection Object Reference=Get MPC_Audio_Reactive_Parameters() -> (MPC_Audio_Reactive_Parameters:Material Parameter Collection Object Reference)
    ParameterName:Name=Get Amplitude() -> (Amplitude:Name=None)
   ) -> (then, ReturnValue:float) {is=title K2Node_CallMaterialParameterCollectionFunction}
   > Set ReactiveToMusic(ReactiveToMusic:float=@n14) -> (Output_Get:float)
   TryGetPawnOwner()
  ]}
]
variables [
 alpha {type=number, default=0.75}
 Amplitude {type=string, default=Amplitude_Music}
 MPC_Audio_Reactive_Parameters {type=MaterialParameterCollection, default=/Game/Athena/Events/AudioAnalysis/MPC_Audio_Reactive_Parameters}
 Range {type=number, default=1}
 ReactiveToMusic {type=number}
 Speakers_01 {type=Vector, default=<0.5101 0 0>}
 Speakers_02 {type=Vector, default=<0.4725 0 0>}
 Speakers_03 {type=Vector, default=<0.2481 0 0>}
 Speakers_04 {type=Vector}
 Speakers_05 {type=Vector}
 Speakers_06 {type=Vector}
 "Translate Shard 1 " {type=Vector, default=<0.6282 0 0>}
 Translate Shard 2 {type=Vector, default=<-0.9689 0 0>}
 Translate Shard 3 {type=Vector, default=<-0.3153 0 0>}
 Translate Shard 4 {type=Vector, default=<0.6282 0 0>}
 Translate Shard 5 {type=Vector, default=<0.9941 0 0>}
 Translate Shard 6 {type=Vector, default=<0.4461 0 0>}
]
